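# MI_Clear is a translucent, Default Lit base material in the marketing screenshots folder, apparently meant as a clear surface, judging by its name.
# Despite the MI_ prefix, it is a Material rather than a Material Instance.
# Its graph holds only two Constant nodes at their default value, with no textures, parameters or visible connections to material outputs.
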Material /Game/Marketing_Screenshots/Materials/MI_Clear
graph {
 properties {blendMode=BLEND_Translucent, pixelDepthOffsetMode=PDOM_Legacy, usedShadingModels=Default Lit}
 nodes [
  Constant()
  Constant()
 ]
}
